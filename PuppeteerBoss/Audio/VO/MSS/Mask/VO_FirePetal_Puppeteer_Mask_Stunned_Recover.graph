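# One-shot stereo MetaSound Source for the Puppeteer boss Mask's stunned-recover voice line.
# It wraps the shared MSS_VO_BR_Boss_Stereo graph, passing in four recorded takes and returning its audio and finished trigger.
# Delay, probability, pitch and volume controls are exposed but left at neutral defaults.
# It uses Mask-specific attenuation, the critical enemy VO sound class and a Mask combat-chatter concurrency set.
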MetaSoundSource /PuppeteerBoss/Audio/VO/MSS/Mask/VO_FirePetal_Puppeteer_Mask_Stunned_Recover
inputs [
 "Delay Max (Seconds)" {type=Time}
 "Delay Min (Seconds)" {type=Time}
 "Probability (0-1)" {type=Float}
 "Random Pitch Max (Semitones)" {type=Float}
 "Random Pitch Min (Semitones)" {type=Float}
 "Random Volume Change Max (Decibels)" {type=Float}
 "Random Volume Change Min (Decibels)" {type=Float}
 UE.Source.OnPlay {type=Trigger}
 Wave Array {type=WaveAsset:Array}
]
interfaces [UE.OutputFormat.Stereo:1.0, UE.Source.OneShot:1.0, UE.Source:1.1]
outputs [
 "UE.OutputFormat.Stereo.Audio:0" {type=Audio}
 "UE.OutputFormat.Stereo.Audio:1" {type=Audio}
 UE.Source.OneShot.OnFinished {type=Trigger}
]
pages [Default {
  input_defaults {
   "Delay Max (Seconds)"=0
   "Delay Min (Seconds)"=0
   "Probability (0-1)"=1
   "Random Pitch Max (Semitones)"=0
   "Random Pitch Min (Semitones)"=0
   "Random Volume Change Max (Decibels)"=0
   "Random Volume Change Min (Decibels)"=0
   UE.Source.OnPlay=null
   Wave Array [
    /PuppeteerBoss/Audio/VO/Waves/Mask/VO_FirePetal_Puppeteer_Mask_S33_Stunned_Recover_010
    /PuppeteerBoss/Audio/VO/Waves/Mask/VO_FirePetal_Puppeteer_Mask_S33_Stunned_Recover_012
    /PuppeteerBoss/Audio/VO/Waves/Mask/VO_FirePetal_Puppeteer_Mask_S33_Stunned_Recover_013
    /PuppeteerBoss/Audio/VO/Waves/Mask/VO_FirePetal_Puppeteer_Mask_S33_Stunned_Recover_014
   ]
  }
  nodes [MSS_VO_BR_Boss_Stereo_0 = /Game/Sounds/Fort_Human_VO_BR/NPCs/Shared/MSS_VO_BR_Boss_Stereo(
    "Delay Max (Seconds)"=^"Delay Max (Seconds)"
    "Delay Min (Seconds)"=^"Delay Min (Seconds)"
    "Probability (0-1)"=^"Probability (0-1)"
    "Random Pitch Max (Semitones)"=^"Random Pitch Max (Semitones)"
    "Random Pitch Min (Semitones)"=^"Random Pitch Min (Semitones)"
    "Random Volume Change Max (Decibels)"=^"Random Volume Change Max (Decibels)"
    "Random Volume Change Min (Decibels)"=^"Random Volume Change Min (Decibels)"
    Wave Array=^Wave Array
   )]
  links [
   {dst="@UE.OutputFormat.Stereo.Audio:0", src=MSS_VO_BR_Boss_Stereo_0.UE.OutputFormat.Stereo.Audio:0}
   {dst="@UE.OutputFormat.Stereo.Audio:1", src=MSS_VO_BR_Boss_Stereo_0.UE.OutputFormat.Stereo.Audio:1}
   {dst="@UE.Source.OneShot.OnFinished", src=MSS_VO_BR_Boss_Stereo_0.UE.Source.OneShot.OnFinished}
  ]
 }]
properties {
 assetImportData=&self:AssetImportData
 attenuationSettings=&/SA_VO_Puppeteer_Mask
 maxDistance=10600
 numChannels=2
 outputFormat=Stereo
 sampleRate=48000
 soundClassObject=&Fort_VO_Enemy_Critical
 volume=5
 concurrencyOverrides {bLimitToOwner=true, maxCount=1, resolutionRule=PreventNew}
 concurrencySet [/PuppeteerBoss/Audio/VO/NPC_Combat_Chatter_PuppeteerMask_Conc]
}
